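# Post-process-style Animation Blueprint for the hatless Lyrical female character's face accessory part, adding secondary motion to hair.
# It copies the attached parent's pose, chains spring-driven AnimDynamics through hat-off hair and four bangs bones, then runs RigidBody simulation relative to neck_01.
# Each update, a deprecated state function sets the RigidBody gravity override.
# A three-node ponytail AnimDynamics chain with spherical limits exists but is disconnected from the output.
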AnimBlueprint /BRCosmetics/Characters/Player/Female/Medium/Bodies/F_MED_Lyrical_Hat_Off/Meshes/Parts/F_MED_Lyrical_Hat_Off_FaceAcc_AnimBP
target_skeleton=&/F_MED_Lyrical_Hat_Off_FaceAcc_Skeleton
graphs [
 AnimGraph {nodes [
   n1 = Anim Dynamics - Bone: dyn_ponytail_01(Alpha=1, ExternalForce="0,0,0") {
    angularDampingOverride=0.98
    angularSpringConstant=300
    bAngularSpring=true
    bGravityOverrideInSimSpace=true
    bOverrideAngularDamping=true
    bUseSphericalLimits=true
    gravityScale=0.31
    simulationSpace=RootRelative
    alphaBoolBlend {blendOption=Linear}
    boundBone {boneName=dyn_ponytail_01}
    componentAppliedLinearAccClamp=<1000 1000 800>
    componentLinearAccScale=<0.5 0.5 0.25>
    componentLinearVelScale=<0.5 0.5 0.25>
    physicsBodyDefinitions [{
      collisionType=CustomSphere
      sphereCollisionRadius=3
      boundBone {boneName=dyn_ponytail_01}
      boxExtents=<20 20 15>
      constraintSetup {angularTargetAxis=AxisY, twistAxis=AxisZ, angularLimitsMax=<0 5 10>, angularLimitsMin=<0 -5 0>, angularTarget=<0 1 0>}
      localJointOffset=<5 0 0>
     }]
    sphericalLimits [{limitRadius=55, drivingBone {boneName=spine_05}, sphereLocalOffset=<-25.9718 35.4505 0>}]
   }
   n2 = Anim Dynamics - Bone: dyn_ponytail_02(ComponentPose=@n1, Alpha=1, ExternalForce="0,0,0") {
    angularDampingOverride=0.98
    angularSpringConstant=1500
    bAngularSpring=true
    bGravityOverrideInSimSpace=true
    bOverrideAngularDamping=true
    bUseSphericalLimits=true
    gravityScale=0.32
    simulationSpace=RootRelative
    alphaBoolBlend {blendOption=Linear}
    boundBone {boneName=dyn_ponytail_02}
    componentAppliedLinearAccClamp=<1000 1000 1200>
    componentLinearAccScale=<0.75 0.75 0.95>
    componentLinearVelScale=<0.75 0.75 0.95>
    physicsBodyDefinitions [{
      collisionType=CustomSphere
      sphereCollisionRadius=2
      boundBone {boneName=dyn_ponytail_02}
      boxExtents=<15 25 25>
      constraintSetup {angularLimitsMax=<0 5 10>, angularLimitsMin=<0 -5 -60>, angularTarget=<0 0 0>}
      localJointOffset=<5 0 0>
     }]
    sphericalLimits [{limitRadius=55, drivingBone {boneName=spine_05}, sphereLocalOffset=<-31.0875 38.7244 0>}]
   }
   Anim Dynamics - Bone: dyn_ponytail_03(ComponentPose=@n2, Alpha=0.7, ExternalForce="0,0,0") {
    angularDampingOverride=0.98
    angularSpringConstant=1500
    bAngularSpring=true
    bGravityOverrideInSimSpace=true
    bOverrideAngularDamping=true
    bUseSphericalLimits=true
    gravityScale=0.31
    simulationSpace=RootRelative
    alphaBoolBlend {blendOption=Linear}
    boundBone {boneName=dyn_ponytail_03}
    componentAppliedLinearAccClamp=<1000 1000 1200>
    componentLinearAccScale=<0.75 0.75 0.95>
    componentLinearVelScale=<0.75 0.75 0.95>
    physicsBodyDefinitions [{
      collisionType=CustomSphere
      sphereCollisionRadius=3.5
      boundBone {boneName=dyn_ponytail_03}
      boxExtents=<15 25 25>
      constraintSetup {angularLimitsMax=<0 5 8>, angularLimitsMin=<0 -5 -50>, angularTarget=<0 0 0>}
      localJointOffset=<4 0 0>
     }]
    sphericalLimits [
     {limitRadius=56, drivingBone {boneName=spine_05}, sphereLocalOffset=<-24.6882 48.8997 0>}
     {limitRadius=55, drivingBone {boneName=spine_05}, sphereLocalOffset=<-37.921 36.8314 0>}
    ]
   }
   n3 = Anim Dynamics - Bone: dyn_hair_hat_off_01(ComponentPose=Local To Component(LocalPose=Copy Pose From Mesh() {bUseAttachedParent=true}), Alpha=1, ExternalForce="0,0,0") {
    angularDampingOverride=1
    angularSpringConstant=1500
    bAngularSpring=true
    bGravityOverrideInSimSpace=true
    bOverrideAngularDamping=true
    gravityScale=0.29
    simulationSpace=RootRelative
    alphaBoolBlend {blendOption=Linear}
    boundBone {boneName=dyn_hair_hat_off_01}
    physicsBodyDefinitions [{
      sphereCollisionRadius=0
      boundBone {boneName=dyn_hair_hat_off_01}
      boxExtents=<13 20 20>
      constraintSetup {angularLimitsMax=<0 5 0>, angularLimitsMin=<0 -5 -20>, angularTarget=<0 0 0>}
      localJointOffset=<2.5 0 0>
     }]
   }
   n4 = Anim Dynamics - Bone: dyn_bangs_01(ComponentPose=@n3, Alpha=1, ExternalForce="0,0,0") {
    angularDampingOverride=1
    angularSpringConstant=1500
    bAngularSpring=true
    bGravityOverrideInSimSpace=true
    bOverrideAngularDamping=true
    gravityScale=0.29
    simulationSpace=RootRelative
    alphaBoolBlend {blendOption=Linear}
    boundBone {boneName=dyn_bangs_01}
    physicsBodyDefinitions [{
      sphereCollisionRadius=0
      boundBone {boneName=dyn_bangs_01}
      boxExtents=<13 20 20>
      constraintSetup {angularLimitsMax=<0 5 0>, angularLimitsMin=<0 -5 -20>, angularTarget=<0 0 0>}
      localJointOffset=<2.5 0 0>
     }]
   }
   n5 = Anim Dynamics - Bone: dyn_bangs_02(ComponentPose=@n4, Alpha=1, ExternalForce="0,0,0") {
    angularDampingOverride=1
    angularSpringConstant=1500
    bAngularSpring=true
    bGravityOverrideInSimSpace=true
    bOverrideAngularDamping=true
    gravityScale=0.29
    simulationSpace=RootRelative
    alphaBoolBlend {blendOption=Linear}
    boundBone {boneName=dyn_bangs_02}
    physicsBodyDefinitions [{
      sphereCollisionRadius=0
      boundBone {boneName=dyn_bangs_02}
      boxExtents=<13 20 20>
      constraintSetup {angularLimitsMax=<0 0 0>, angularLimitsMin=<0 -5 -20>, angularTarget=<0 0 0>}
      localJointOffset=<2.5 0 0>
     }]
   }
   n6 = Anim Dynamics - Bone: dyn_bangs_03(ComponentPose=@n5, Alpha=1, ExternalForce="0,0,0") {
    angularDampingOverride=1
    angularSpringConstant=1500
    bAngularSpring=true
    bGravityOverrideInSimSpace=true
    bOverrideAngularDamping=true
    gravityScale=0.29
    simulationSpace=RootRelative
    alphaBoolBlend {blendOption=Linear}
    boundBone {boneName=dyn_bangs_03}
    physicsBodyDefinitions [{
      sphereCollisionRadius=0
      boundBone {boneName=dyn_bangs_03}
      boxExtents=<13 20 20>
      constraintSetup {angularLimitsMax=<0 5 0>, angularLimitsMin=<0 -5 -20>, angularTarget=<0 0 0>}
      localJointOffset=<2.5 0 0>
     }]
   }
   n7 = Anim Dynamics - Bone: dyn_bangs_04(ComponentPose=@n6, Alpha=1, ExternalForce="0,0,0") {
    angularDampingOverride=1
    angularSpringConstant=1500
    bAngularSpring=true
    bGravityOverrideInSimSpace=true
    bOverrideAngularDamping=true
    gravityScale=0.29
    simulationSpace=RootRelative
    alphaBoolBlend {blendOption=Linear}
    boundBone {boneName=dyn_bangs_04}
    physicsBodyDefinitions [{
      sphereCollisionRadius=0
      boundBone {boneName=dyn_bangs_04}
      boxExtents=<13 20 20>
      constraintSetup {angularLimitsMax=<0 5 0>, angularLimitsMin=<0 -5 -20>, angularTarget=<0 0 0>}
      localJointOffset=<2.5 0 0>
     }]
   }
   n8 = RigidBody!(ComponentPose=@n7, Alpha, OverrideWorldGravity:Vector=Get Gravity Override(), ExternalForce) -> (Pose) {
    bClampLinearTranslationLimitToRefPose=true
    bOverrideWorldGravity=true
    overridePhysicsAsset=&/F_MED_Lyrical_Hat_Off_FaceAcc_Physics
    simulationSpace=BaseBoneSpace
    alphaBoolBlend {blendOption=Linear}
    baseBoneRef {boneName=neck_01}
    componentAppliedLinearAccClamp=<3000 3000 3000>
    componentLinearAccScale=<0.4 0.3 0.3>
    componentLinearVelScale=<0.4 0.3 0.3>
    overrideWorldGravity {z=-980}
    simSpaceSettings {maxAngularAcceleration=1000, maxAngularVelocity=3, maxLinearAcceleration=250, maxLinearVelocity=500, velocityScaleZ=0, worldAlpha=1}
   }
   Output Pose(Result=Component To Local(ComponentPose=@n8)) {layerGroup=None}
  ]}
 EventGraph {nodes [
   Event BlueprintUpdateAnimation()
   > n1 = Gravity Override State Function DEPRECATED(CharacterPartAnimInstance=Self-Reference(), GameTimeInSeconds=GetGameTimeInSeconds())
   > Set Gravity Override(Gravity Override=@n1)
  ]}
]
variables [Gravity Override {type=Vector}]
